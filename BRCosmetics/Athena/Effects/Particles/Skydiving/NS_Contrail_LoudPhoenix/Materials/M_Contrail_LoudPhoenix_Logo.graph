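# Translucent unlit Niagara material for the LoudPhoenix skydiving contrail, drawing its glowing logo or symbols.
# Opacity layers logo lines with a sphere energy mask, scrolling smoke noise, and glow/blur textures, scaled by particle alpha and depth fade.
# Use_SubUV blends logo to symbol texture, rotating noise to panned, and sine-wave vertex offset to Sprite-function offset.
# Particle color feeds emissive adjustments; dynamic parameters drive wave timing, strength and sprite scale.
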Material /BRCosmetics/Athena/Effects/Particles/Skydiving/NS_Contrail_LoudPhoenix/Materials/M_Contrail_LoudPhoenix_Logo
graph {
 properties {
  bUsedWithInstancedStaticMeshes=true
  bUsedWithNiagaraMeshParticles=true
  bUsedWithNiagaraSprites=true
  blendMode=BLEND_Translucent
  pixelDepthOffsetMode=PDOM_Legacy
  shadingModel=MSM_Unlit
  usedShadingModels=Unlit
 }
 nodes [
  n1 = TextureCoordinate() {vTiling=1.05}
  n2 = Frac(None=Multiply(A=TextureCoordinate() {coordinateIndex=2, vTiling=1.05}, B=Constant2Vector() {g=1, r=8}))
  n3 = ScalarParameter() {parameterName=Use_SubUV}
  n4 = Multiply(
   A=SphereMask(
    A=LinearInterpolate(A=@n1, B=@n2, Alpha=@n3)
    B=Constant2Vector() {g=0.5, r=0.5}
    Radius=ScalarParameter() {defaultValue=0.4, parameterName=SphereMask_Radius}
    Hardness=ScalarParameter() {defaultValue=0.5, parameterName=SphereMask_Hardness}
   ) {attenuationRadius=0.4, hardnessPercent=0.5}
   B=ScalarParameter() {defaultValue=0.1, parameterName=EnergyMask Opacity}
  )
  Add(A=@n4)
  n5 = DynamicParameter() {paramNames [AmplitudeStrenght, Time, Sprite Scale, WPOStrenght]}
  n6 = MaterialFunctionCall!(Center Location=WorldPosition().XYZ, 0-1 UVs, XY scale=@n5.Sprite Scale, OverridePivot, "Normals (Optional)") {materialFunction=&Sprite}
  Add(A=Multiply(A=VertexNormalWS()) {constB=0.125}, B=@n6.World position offset)
  Constant2Vector() {g=10, r=10}
  Constant()
  DynamicParameter() {parameterIndex=1, paramNames [Energy, DetailExponent, PanAdd, RandomPan]}
  n7 = Constant3Vector() {constant=<rgba 0 0 1 0>}
  n8 = LinearInterpolate(A=ScalarParameter() {defaultValue=8, parameterName=WPOStrenght}, Alpha=@n5.WPOStrenght) {constA=5, constB=0}
  n9 = TextureCoordinate()
  n10 = Constant() {r=0.5}
  n11 = Add(A=Multiply(A=SphereMask(A=@n9, B=@n10) {attenuationRadius=0.5, hardnessPercent=0}) {constB=1.5}, B=Multiply(A=Time(), B=@n5.Time) {constB=2})
  LinearInterpolate(
   A=Multiply(A=Multiply(A=@n7, B=@n8) {constB=5}, B=Sine(None=@n11))
   B=@n6.World position offset
   Alpha=ScalarParameter() {parameterName=Use_SubUV}
  )
  n12 = ParticleColor()
  MaterialFunctionCall(Emissive=@n12.RGB) {materialFunction=&MF_FX_EmissiveAdjustments}
  n13 = TextureSampleParameter2D() {
   parameterName=Texture1
   samplerType=SAMPLERTYPE_LinearColor
   texture=/BRCosmetics/Athena/Effects/Particles/Skydiving/NS_Contrail_LoudPhoenix/Textures/T_LoudPhoenix_Logo2_Lines
  }
  n14 = Multiply(A=@n4, B=@n13.RGB)
  n15 = TextureCoordinate()
  n16 = Constant2Vector() {g=0.5, r=0.5}
  n17 = Multiply(A=TextureCoordinate(), B=ScalarParameter() {defaultValue=0.5, group=Tiling Texture, parameterName=Rotator UV 1, sortPriority=3})
  n18 = Multiply(A=Time(), B=ScalarParameter() {defaultValue=-3, parameterName=Speed1})
  n19 = TextureCoordinate()
  n20 = TextureSampleParameter2D(UVs=LinearInterpolate(
    A=Rotator(Coordinate=@n17, Time=@n18)
    B=Panner(Coordinate=@n19) {speedX=-0.5, speedY=-0.3}
    Alpha=ScalarParameter() {parameterName=Use_SubUV}
   )) {group=Tiling Texture, parameterName=TilingTexture1, texture=&T_Inky_Smoke_Tile_Chromatic}
  n21 = Multiply(
   A=TextureCoordinate() {uTiling=2, vTiling=2}
   B=ScalarParameter() {defaultValue=0.5, group=Tiling Texture, parameterName=Rotator UV 2, sortPriority=3}
  )
  n22 = Multiply(A=Time(), B=ScalarParameter() {defaultValue=3, parameterName=Speed2})
  n23 = TextureCoordinate()
  n24 = TextureSampleParameter2D(UVs=LinearInterpolate(
    A=Rotator(Coordinate=@n21, Time=@n22)
    B=Panner(Coordinate=@n23) {speedX=0.15, speedY=0.65}
    Alpha=ScalarParameter() {parameterName=Use_SubUV}
   )) {group=Tiling Texture, parameterName=TilingTexture1, texture=&T_Inky_Smoke_Tile_Chromatic}
  n25 = Multiply(A=@n20.G, B=@n24.B)
  n26 = Multiply(A=Power(Base=SphereMask(A=@n15, B=@n16) {attenuationRadius=0.75, hardnessPercent=0}) {constExponent=3}, B=@n25)
  n27 = TextureSampleParameter2D() {parameterName=GlowTexture, texture=/BRCosmetics/Athena/Effects/Particles/Skydiving/NS_Contrail_LoudPhoenix/Textures/T_LoudPhoenix_Logo2_Glow}
  n28 = Power(Base=@n27.RGB) {constExponent=1.5}
  n29 = TextureSampleParameter2D() {parameterName=BlurTexture, texture=/BRCosmetics/Athena/Effects/Particles/Skydiving/NS_Contrail_LoudPhoenix/Textures/T_LoudPhoenix_Logo2_Blur}
  n30 = Add(
   A=Add(A=Power(Base=@n14) {constExponent=3}, B=Multiply(A=@n13.RGB, B=@n26))
   B=Multiply(A=Multiply(A=@n28, B=@n13.RGB), B=Power(Base=@n29.RGB) {constExponent=0.5})
  )
  n31 = TextureSample(UVs=TextureCoordinate() {coordinateIndex=2}) {texture=/BRCosmetics/Athena/Effects/Particles/Skydiving/NS_Contrail_LoudPhoenix/Textures/T_LoudPhoenix_Symbols1}
  n32 = Add(A=Power(Base=Multiply(A=@n4, B=@n31.RGB)) {constExponent=5}, B=Multiply(A=@n31.RGB, B=@n25))
  n33 = ScalarParameter() {parameterName=Use_SubUV}
  Multiply(
   A=Multiply(A=@n12.A, B=LinearInterpolate(A=@n30, B=@n32, Alpha=@n33))
   B=DepthFade(FadeDistance=ScalarParameter() {defaultValue=2, parameterName=DepthFade})
  )
  Multiply(A=@n27.RGB, B=ScalarParameter() {defaultValue=1, parameterName=Glow Opacity})
  ScalarParameter() {defaultValue=20, parameterName=SpriteScale}
 ]
}
