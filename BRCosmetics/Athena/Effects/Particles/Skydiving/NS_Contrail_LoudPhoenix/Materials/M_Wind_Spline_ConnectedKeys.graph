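# Unlit additive material for Niagara mesh particles, drawing wispy wind splines in the Loud Phoenix skydiving contrail.
# World position offset adds SplineThicken, with width shrunk near the camera, to a layered-sine wobble masked along V.
# Panning smoke masks, particle alpha, gradient masks, depth fade and optional end fade shape opacity, while a chromatic smoke texture and secondary tint color it.
# Dynamic parameters drive wobble, time offset, mask position and width per particle.
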Material /BRCosmetics/Athena/Effects/Particles/Skydiving/NS_Contrail_LoudPhoenix/Materials/M_Wind_Spline_ConnectedKeys
graph {
 properties {
  bUsedWithNiagaraMeshParticles=true
  blendMode=BLEND_Additive
  pixelDepthOffsetMode=PDOM_Legacy
  shadingModel=MSM_Unlit
  twoSided=true
  usedShadingModels=Unlit
 }
 nodes [
  n1 = ScalarParameter() {defaultValue=0.25, parameterName=SplineWidthTipMult}
  n2 = DynamicParameter() {paramNames [WobbleMult,RandomAdd,PanAdd,SplineWidth]}
  n3 = MaterialFunctionCall(Distance=ScalarParameter() {defaultValue=100, parameterName=NearCameraShrink}) {materialFunction=&Near_Camera_Fade}
  n4 = Multiply(A=@n2.SplineWidth, B=Multiply(A=ScalarParameter() {defaultValue=8, parameterName=SplineWidth}, B=Power(Base=@n3.Result) {constExponent=3}))
  n5 = TextureCoordinate()
  n6 = ComponentMask(None=@n5) {g=true}
  n7 = Constant() {r=0.5}
  n8 = TextureCoordinate()
  n9 = Multiply(A=Add(A=@n2.RandomAdd, B=Time()), B=ScalarParameter() {defaultValue=1.5, parameterName=WobbleSpeed})
  n10 = ComponentMask(None=Add(A=@n8, B=Multiply(A=@n9) {constB=-0.6})) {g=true}
  n11 = Sine(None=@n10) {period=0.2}
  n12 = ComponentMask(None=Add(A=@n8, B=Multiply(A=@n9) {constB=-1})) {g=true}
  n13 = AppendVector(A=Constant(), B=Add(A=Multiply(A=@n11) {constB=0.2}, B=Sine(None=@n12) {period=0.8}))
  n14 = Add(A=Multiply(A=Sine(None=@n10) {period=0.23}) {constB=0.2}, B=Sine(None=@n12) {period=0.6})
  n15 = Normalize(VectorInput=Transform(None=AppendVector(A=@n13, B=@n14)) {transformSourceType=TRANSFORMSOURCE_Local})
  n16 = Multiply(A=ScalarParameter() {defaultValue=2, parameterName=Wobble}, B=@n2.WobbleMult)
  Add(
   A=MaterialFunctionCall(WidthBase=Multiply(A=@n1, B=@n4) {constB=5}, WidthTip=@n4, UVs for Thickness=Power(Base=@n5) {constExponent=0.6}) {materialFunction=&SplineThicken}.WorldPositionOffset
   B=Multiply(A=SphereMask(A=@n6, B=@n7, Radius=@n7) {hardnessPercent=1}, B=Multiply(A=@n15, B=@n16))
  )
  n17 = SphereMask(A=TextureCoordinate(), B=Constant2Vector() {r=0.5}) {attenuationRadius=0.5, hardnessPercent=0}
  n18 = Power(Base=@n17) {constExponent=4}
  n19 = ScalarParameter() {defaultValue=1, parameterName=Edge-Trailing_GradMult}
  n20 = ScalarParameter() {defaultValue=1, parameterName=Edge-Leading_GradMult}
  n21 = ParticleColor()
  n22 = OneMinus(None=ComponentMask(None=TextureCoordinate() {coordinateIndex=1}) {r=true})
  n23 = ScalarParameter() {defaultValue=2.5, parameterName=X-Tiling}
  n24 = Constant() {r=1}
  n25 = Panner(
   Coordinate=Multiply(A=TextureCoordinate() {coordinateIndex=1}, B=AppendVector(A=@n23, B=@n24))
   Time=Multiply(A=Time(), B=ScalarParameter() {defaultValue=3, parameterName=PanSpeed})
  ) {speedX=0.5}
  n26 = TextureSampleParameter2D(UVs=@n25) {parameterName=Mask, samplerType=SAMPLERTYPE_Grayscale, texture=&T_Smoke_Trail_Soft}
  n27 = Add(A=Multiply(A=@n21.A, B=Add(A=@n22) {constB=0.1}), B=@n26.R)
  n28 = Clamp(None=Multiply(A=LinearInterpolate(A=@n19, B=@n20, Alpha=@n27) {constA=-3}, B=@n26.R))
  n29 = ComponentMask(None=TextureCoordinate()) {g=true}
  n30 = Constant() {r=0.5}
  n31 = SphereMask(A=@n29, B=@n30, Radius=@n30) {attenuationRadius=0.5, hardnessPercent=90}
  n32 = SphereMask(A=@n29, B=LinearInterpolate(Alpha=@n2.PanAdd) {constA=-1, constB=0.25}) {attenuationRadius=1, hardnessPercent=50}
  n33 = Multiply(A=@n21.A, B=Multiply(A=Add(A=@n18, B=@n28), B=Multiply(A=@n31, B=@n32) {constA=1}))
  n34 = ScalarParameter() {defaultValue=20, parameterName=DepthFadeDist}
  n35 = Multiply(
   A=@n33
   B=ShadingPathSwitch(Default=DepthFade(Opacity=@n33, FadeDistance=@n34) {fadeDistanceDefault=55}, Mobile=Constant() {r=1}) {desc=Migrated from FeatureLevelSwitch}
  )
  n36 = TextureCoordinate()
  n37 = ScalarParameter() {defaultValue=1, parameterName=EndFadeMult}
  n38 = Clamp(None=Multiply(A=ComponentMask(None=@n36) {r=true}, B=@n37))
  LinearInterpolate(A=@n35, B=Multiply(A=@n35, B=Power(Base=@n38, Exp=@n37)), Alpha=ScalarParameter() {parameterName=UseEndFade?})
  n39 = Desaturation(
   None=TextureSampleParameter2D(UVs=@n25) {parameterName=ColorTexture, texture=/Game/Effects/Fort_Effects/Textures/Smoke/T_Smoke_Trail_Soft_Chromatic}.RGB
   Fraction=ScalarParameter() {parameterName=Desaturate-ColorTexture}
  )
  n40 = Multiply(
   A=LinearInterpolate(A=VectorParameter() {parameterName=SeondaryColor, defaultValue=<rgba 1 0.2448 0.2985 1>}.RGB, Alpha=@n17)
   B=Add(A=Multiply(A=@n21.RGB, B=@n39), B=Multiply(A=@n21.RGB, B=@n18))
  )
  MaterialFunctionCall(Emissive=@n40) {materialFunction=&MF_FX_EmissiveAdjustments}
  Reroute(None=@n40)
 ]
}
