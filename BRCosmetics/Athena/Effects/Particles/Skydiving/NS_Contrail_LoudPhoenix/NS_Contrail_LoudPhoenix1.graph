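# Niagara skydiving contrail for the LoudPhoenix cosmetic, combining bone-attached ribbons, forward smoke, a logo mesh under the player, rotating key meshes and center glows.
# Ribbons are looping burst particles re-pinned to skeletal sockets each update and stretched along dive velocity, rather than trailing behind.
# System dive and curve values drive ribbon length, smoke spawning, key orbits and glow fades.
# Front-end modifiers swap gameplay directions, offsets and intensities for menu display.
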NiagaraSystem /BRCosmetics/Athena/Effects/Particles/Skydiving/NS_Contrail_LoudPhoenix/NS_Contrail_LoudPhoenix1
dependencies {
 usedDataInterfaces [
  &NiagaraDataInterfaceCurve
  &NiagaraDataInterfaceColorCurve
  &NiagaraDataInterfaceMeshRendererInfo
  &NiagaraDataInterfaceVector2DCurve
  &NiagaraDataInterfaceArrayFloat3
  &NiagaraDataInterfaceArrayInt32
  &NiagaraDataInterfaceArrayFloat
 ]
 usedDynamicInputs [
  /Niagara/DynamicInputs/Rotation/InvertQuaternion
  &Add_Vector
  &self:Abs_Z_Axis_01
 ]
 usedModules [
  &self:SystemSpawnScript.SystemScriptSource.SystemScriptGraph.NiagaraNodeAssignment_0.SetVariables_#1
  &self:SystemSpawnScript.SystemScriptSource.SystemScriptGraph.NiagaraNodeAssignment_6.SetVariables_#2
  &SystemState
  &self:SystemSpawnScript.SystemScriptSource.SystemScriptGraph.NiagaraNodeAssignment_3.SetVariables_#3
  /Game/Effects/Niagara/Modules/Skydiving/FN_DivingAttributes/v5/FN_DivingAttributes
  &self:SystemSpawnScript.SystemScriptSource.SystemScriptGraph.NiagaraNodeAssignment_2.SetVariables_#4
  &self:SystemSpawnScript.SystemScriptSource.SystemScriptGraph.NiagaraNodeAssignment_1.SetVariables_#5
  &FN_Integrator
  &InterpolateOverTime
  &Timeline
  &self:SystemSpawnScript.SystemScriptSource.SystemScriptGraph.NiagaraNodeAssignment_4.SetVariables_#6
  &self:SystemSpawnScript.SystemScriptSource.SystemScriptGraph.NiagaraNodeAssignment_5.SetVariables_#7
 ]
}
emitters {
 CenterGlow_for_Keys {
  emitterUpdateScript [
   &EmitterState(Life Cycle Mode=::NewEnumerator1, Loop Behavior=::NewEnumerator1, Loop Duration Mode=::NewEnumerator1, UseLoopDelay=false)
   &SpawnBurst_Instantaneous(Use Spawn Probability=false, UseLoopCountLimit=false)
  ]
  particleSpawnScript [
   &PartitionParticles()
   /SetVariables(Particles.SpawnGroup=$Particles.PartitionParticles.Partition) {isSetParameters=true}
   &InitializeParticle(
    Color Mode=::NewEnumerator0
    Lifetime=5
    Sprite Size Mode=::NewEnumerator3
    Uniform Sprite Size=75
    UsePositionOffset=false
    Write Material Random=true
   )
  ]
  particleUpdateScript [
   /SetVariables(Particles.Sprite_HaloColor:LinearColor=&MultiplyLinearColor(
     Linear Color=<rgba 0.5052 1 0 1>
     Scale Factor=&MakeVector4(
      X=1
      Y=1
      Z=1
      W=&FloatFromCurve(
       FloatCurve=DataInterface() {propertyValues {Curve {keys [{interpMode=RCIM_Cubic, tangentMode=RCTM_User, time=0, value=1}, {interpMode=RCIM_Cubic, tangentMode=RCTM_User, time=1, value=0}]}}}
       CurveIndex=$System.ScalingCurve3
      )
     )
    )) {isSetParameters=true}
   &ShapeLocation(
    MaskSpawnGroup=false
    Offset=$System.Sphere_RotateVector
    Offset Mode=::NewEnumerator0
    Rotation Axis=<0 0 1>
    Rotation Mode=::NewEnumerator1
    Sphere Distribution=::NewEnumerator1
    Sphere Radius Position=0
    UseOwnerScale=true
   ) {module=ShapeLocation002}
   &ParticleState(
    Kill Particles When Lifetime Has Elapsed=false
    Let Infinitely Lived Particles Die When Emitter Deactivates=true
    Loop Particles Lifetime=true
   )
   &ScaleColor(
    Scale Alpha=&FloatFromCurve(
     FloatCurve=DataInterface() {propertyValues {Curve {keys [{interpMode=RCIM_Cubic, tangentMode=RCTM_User, time=0, value=1}, {interpMode=RCIM_Cubic, tangentMode=RCTM_User, time=0.25, value=0}]}}}
     CurveIndex=$System.ScalingCurve3
    )
    Scale RGB=&MakeVectorFromLinearColorRGB(Color=&ColorFromCurve(ColorCurve=DataInterface() {propertyValues {
        AlphaCurve {keys [<at 0 1 RCIM_Cubic>,<at 1 1 RCIM_Cubic>]}
        BlueCurve {keys [<at 0.0113 0>,<at 0.309 0 RCIM_Cubic>,<at 0.9977 0>]}
        GreenCurve {keys [<at 0.0113 1.3052>,<at 0.309 0.5155 RCIM_Cubic>,<at 0.9977 1.3052>]}
        RedCurve {keys [<at 0.0113 5>,<at 0.309 5 RCIM_Cubic>,<at 0.9977 5>]}
       }}))
    ScaleA=true
    ScaleRGB=true
   )
   &FN_Light_Attributes(
    Color=<rgba 3 1.5145 0 1>
    Create Color Attribute=true
    Create Exponent Attribute=true
    Create Radius Attribute=true
    Exponent=3
    Radius=75
    Scale Color=&FloatFromCurve(
     FloatCurve=DataInterface() {propertyValues {Curve {keys [{interpMode=RCIM_Cubic, tangentMode=RCTM_User, time=0, value=1}, {interpMode=RCIM_Cubic, tangentMode=RCTM_User, time=0.5, value=0}]}}}
     CurveIndex=$System.ScalingCurve2
     Scale Curve=3
    )
   )
  ]
  rendererClasses [&NiagaraSpriteRendererProperties, &NiagaraLightRendererProperties]
  renderers [
   {rendererClass=&NiagaraSpriteRendererProperties}
   {rendererClass=&NiagaraSpriteRendererProperties, rendererIndex=1}
   {rendererClass=&NiagaraLightRendererProperties, rendererIndex=2}
  ]
 }
 CenterGlow_for_Keys001 {
  emitterUpdateScript [
   &EmitterState(Life Cycle Mode=::NewEnumerator1, Loop Behavior=::NewEnumerator1, Loop Duration Mode=::NewEnumerator1, UseLoopDelay=false)
   &SpawnBurst_Instantaneous(Spawn Count=40, Use Spawn Probability=false, UseLoopCountLimit=false) {module=SpawnBurst_Instantaneous001}
   &SpawnBurst_Instantaneous(Spawn Group=1, Use Spawn Probability=false, UseLoopCountLimit=false)
  ]
  particleSpawnScript [
   /SetVariables(Particles.VisibilityTag=&SpawnGroup()) {isSetParameters=true}
   /SetVariables(Particles.RibbonLinkOrder:NiagaraFloat=&ReturnNormalizedExecIndex()) {isSetParameters=true}
   &InitializeParticle(
    Color Mode=::NewEnumerator0
    Lifetime=5
    Mesh Scale=<75 25 1>
    Mesh Scale Mode=::NewEnumerator4
    Ribbon Width=10
    Ribbon Width Mode=::NewEnumerator1
    Sprite Size Mode=::NewEnumerator3
    Uniform Sprite Size=75
    UsePositionOffset=false
    Write Material Random=true
   )
   &InitialMeshOrientation(
    Mesh Axis To Orient=<0 0 1>
    Mesh Orientation Mode=::NewEnumerator2
    Orientation Vector=<0 0 1>
    Rotation=<0 0 0.5>
    Use Rotation Vector=false
   )
  ]
  particleUpdateScript [
   &ParticleState(
    Kill Particles When Lifetime Has Elapsed=false
    Let Infinitely Lived Particles Die When Emitter Deactivates=true
    Loop Particles Lifetime=true
   )
   /SetVariables(Particles.DynamicMaterialParameter_Ribbon:Vector4f=&MakeVector4(
     X=0
     Y=-1
     Z=-0.8
     W=&FloatFromCurve(FloatCurve=DataInterface() {propertyValues {Curve {keys [<at 0 1>,<at 1 0>]}}}, CurveIndex=$Particles.RibbonLinkOrder, Scale Curve=0.25)
    )) {isSetParameters=true}
   /SetVariables(Particles.MeshOrientation=&OrientMeshToVector(
     Look At Direction=&FN_FrontEndVector3Modifier(
      Gameplay=&Multiply_VectorByFloat(
       Vector=&NormalizeVector(
        Vector To Normalize=/Niagara/DynamicInputs/Vectors/NegateVector(Vector:Vector3f=&TransformVector(
          Vector=&Multiply_Vector(A=$Engine.Owner.Velocity, B=<0.4 0.4 1>)
          Source Space=::World
          Destination Space=::Local
         ))
        Safe Normalize=true
        Safe Fallback Threshold=0
        Safe Fallback Vector=<0 0 0.001>
       )
       Float=-1
      )
      Front End=&Multiply_VectorByFloat(
       Vector=&NormalizeVector(
        Vector To Normalize=&NegateVector(Vector:Vector3f=&TransformVector(
          Vector=&Multiply_Vector(A=$Engine.Owner.Velocity, B=<0.6 0.6 1>)
          Source Space=::World
          Destination Space=::Local
         ))
        Safe Normalize=true
        Safe Fallback Threshold=0
        Safe Fallback Vector=<0 0 0.001>
       )
       Float=-1
      )
     )
     Reference Axis Vector=<1 0 0>
    )) {isSetParameters=true}
   &ShapeLocation(
    MaskSpawnGroup=false
    Offset=$System.Sphere_RotateVector
    Offset Mode=::NewEnumerator0
    Sphere Distribution=::NewEnumerator1
    Sphere Radius Position=0
    UseOwnerScale=true
   ) {module=ShapeLocation002}
   /SetVariables(Particles.Position=&MaskVectorBySpawnGroup(
     Masked Vector=&AddVectorToPosition(
      Position=$Particles.Position
      Vector=&RotateVectorByQuaternion(
       Vector To Rotate=&Multiply_VectorByFloat(
        Vector=&FN_FrontEndVector3Modifier(Gameplay=$System.DiveVelocityScale, Front End=&Lerp_Vector(A=<0 0 -1>, B=<0 0.25 -1>, Alpha=$System.Dive))
        Float=&FloatFromCurve(
         FloatCurve=DataInterface() {propertyValues {Curve {keys [{tangentMode=RCTM_User, time=0, value=0}, {tangentMode=RCTM_User, time=1, value=1}]}}}
         CurveIndex=&ReturnNormalizedExecIndex()
         Scale Curve=&FloatFromCurve(
          FloatCurve=DataInterface() {propertyValues {Curve {keys [{interpMode=RCIM_Cubic, tangentMode=RCTM_User, time=0.5, value=1}, {interpMode=RCIM_Cubic, tangentMode=RCTM_User, time=1, value=0}]}}}
          CurveIndex=$System.ScalingCurve3
          Scale Curve=-65
         )
        )
       )
       Quaternion=&MultiplyQuaternion(
        Quaternion A=$Particles.MeshOrientation
        Quaternion B=&MakeQuatFromAxisAngle(Axis=<0 1 0>, AngleInDegrees=-90, NormalizeAxis=true)
       )
      )
     )
     Passthrough Vector=Linked() {linkedVariable {name=Particles.Position, type=NiagaraPosition}}
    )) {isSetParameters=true}
   &ScaleColor(
    Scale Alpha=&MaskFloatBySpawnGroup(
     Masked Float=&FloatFromCurve(
      FloatCurve=DataInterface() {propertyValues {Curve {keys [{interpMode=RCIM_Cubic, tangentMode=RCTM_User, time=0, value=1}, {interpMode=RCIM_Cubic, tangentMode=RCTM_User, time=0.5, value=0}]}}}
      CurveIndex=$System.ScalingCurve3
     )
     Passthrough Float=&FloatFromCurve(
      FloatCurve=DataInterface() {propertyValues {Curve {keys [{interpMode=RCIM_Cubic, tangentMode=RCTM_User, time=0, value=1}, {interpMode=RCIM_Cubic, tangentMode=RCTM_User, time=0.5, value=0}]}}}
      CurveIndex=$System.ScalingCurve3
      Scale Curve=&Waveform("Waveform [1]"=::NewEnumerator9, "[1] Phase Scale"=2, "[1] Frequency"=5, "[1] Amplitude Min/Max"=<0.25 1>)
     )
    )
    Scale RGB=&MakeVectorFromLinearColorRGB(Color=&ColorFromCurve(
      ColorCurve=DataInterface() {propertyValues {
        AlphaCurve {keys [<at 0 1>,<at 1 1>]}
        BlueCurve {keys [<at 0 0>,<at 1 0>]}
        GreenCurve {keys [<at 0 0.775>,<at 1 0.3378>]}
        RedCurve {keys [<at 0 2>,<at 1 1>]}
       }}
      CurveIndex=$Particles.RibbonLinkOrder
     ))
    ScaleA=true
    ScaleRGB=true
   )
  ]
  rendererClasses [
   &NiagaraSpriteRendererProperties
   &NiagaraMeshRendererProperties
   &NiagaraRibbonRendererProperties
  ]
  renderers [
   {rendererClass=&NiagaraSpriteRendererProperties}
   {rendererClass=&NiagaraMeshRendererProperties, rendererIndex=1}
   {rendererClass=&NiagaraRibbonRendererProperties, rendererIndex=2}
  ]
 }
 Empty {
  emitterUpdateScript [
   &EmitterState()
   &SpawnBurst_Instantaneous(Spawn Count=40, Use Spawn Probability=false, UseLoopCountLimit=true)
  ]
  particleSpawnScript [
   &PartitionParticles(Write Partition to Ribbon ID=::NewEnumerator0)
   /SetVariables(Particles.RibbonLinkOrder:NiagaraFloat=&ReturnNormalizedExecIndex()) {isSetParameters=true, module=SetVariables_#8}
   &InitializeParticle(
    Color=<rgba 0.2387 1 0 1>
    Lifetime=5
    Position Offset=<0 0 25>
    Ribbon Width=23
    Ribbon Width Mode=::NewEnumerator1
    Sprite Size Mode=::NewEnumerator3
    Uniform Sprite Size=5
    UsePositionOffset=false
    Write Material Random=true
   )
   &SkeletalMeshLocation(
    Bone / Socket Index=&SelectIntFromArray(Direct Array Index=$Particles.PartitionParticles.Partition, Int Selection Array=DataInterface() {propertyValues {IntData [2,3]}})
    Bone Sampling Mode=::NewEnumerator4
    MaskSpawnGroup=false
    Sampled Position Offset Coordinate Space:ENiagaraSkelMeshTransforms=::NewEnumerator2
    Skeletal Mesh=$User.SkeletalMesh
   )
  ]
  particleUpdateScript [
   &ParticleState(
    Kill Particles When Lifetime Has Elapsed=false
    Let Infinitely Lived Particles Die When Emitter Deactivates=true
    Loop Particles Lifetime=true
   )
   &ScaleSpriteSize()
   /Niagara/Modules/Update/Utility/IncrementOverTime(Rate of Change:NiagaraFloat=&Lerp_Float(A=0.2, B=0.9, Alpha=$System.Dive))
   &DynamicMaterialParameters(
    Index 0 Param 1=1
    Index 0 Param 2=&MakeCustomFloatFromBool(
     Bool=&SetBoolByIntComparison(A=$Particles.PartitionParticles.Partition, B=0, Comparison Type=::NewEnumerator0)
     True Float=-0.8
     False Float=-1
    )
    Index 0 Param 3=&MakeCustomFloatFromBool(
     Bool=&SetBoolByIntComparison(A=$Particles.PartitionParticles.Partition, B=0, Comparison Type=::NewEnumerator0)
     True Float=-0.6
     False Float=-0.8
    )
    Index 0 Param 4=&FloatFromCurve(FloatCurve=DataInterface() {propertyValues {Curve {keys [<at 0 1>,<at 1 0>]}}}, CurveIndex=$Particles.RibbonLinkOrder, Scale Curve=0.15)
    Module.Param0WriteEnabled=true
    Module.Param1WriteEnabled=true
    Module.Param2WriteEnabled=true
    Module.Param3WriteEnabled=true
   )
   &SkeletalMeshLocation(
    Bone / Socket Index=&SelectIntFromArray(Direct Array Index=$Particles.PartitionParticles.Partition, Int Selection Array=DataInterface() {propertyValues {IntData [2,3]}})
    Bone Sampling Mode=::NewEnumerator4
    MaskSpawnGroup=false
    Sampled Position Offset Coordinate Space:ENiagaraSkelMeshTransforms=::NewEnumerator2
    Skeletal Mesh=$User.SkeletalMesh
   ) {module=SkeletalMeshLocation001}
   /SetVariables(Particles.Position=&AddVectorToPosition(
     Position=$Particles.Position
     Vector=&Multiply_VectorByFloat(
      Vector=$System.DiveVelocityScale
      Float=&FloatFromCurve(
       FloatCurve=DataInterface() {propertyValues {Curve {keys [{tangentMode=RCTM_User, time=0, value=0}, {tangentMode=RCTM_User, time=1, value=1}]}}}
       CurveIndex=&ReturnNormalizedExecIndex()
       Scale Curve=&Lerp_Float(A=-130, B=-160, Alpha=$System.Dive)
      )
     )
    )) {isSetParameters=true, module=SetVariables_#9}
   /SetVariables(Particles.Position=&AddVectorToPosition(
     Position=$Particles.Position
     Vector=&TransformVector(
      Vector=&Multiply_VectorByFloat(
       Vector=&SelectVectorFromArray(
        Direct Array Index=$Particles.PartitionParticles.Partition
        Vector Selection Array=DataInterface() {propertyValues {FloatData [<0.75 0 0>, <-0.75 0 0>, <0.75 0 0>, <-0.75 0 0>]}}
       )
       Float=&FloatFromCurve(
        FloatCurve=DataInterface() {propertyValues {Curve {keys [
            {arriveTangent=4.1785, interpMode=RCIM_Cubic, leaveTangent=4.1785, tangentMode=RCTM_User, time=0, value=0}
            {arriveTangent=1.1179, interpMode=RCIM_Cubic, leaveTangent=1.1179, tangentMode=RCTM_User, time=0.2042, value=0.6228}
            {arriveTangent=0.0224, interpMode=RCIM_Cubic, leaveTangent=0.0224, tangentMode=RCTM_User, time=1, value=1}
           ]}}}
        CurveIndex=&ReturnNormalizedExecIndex()
        Scale Curve=&MakeCustomFloatFromBool(
         Bool=&SetBoolByIntComparison(A=$Particles.PartitionParticles.Partition, B=1, Comparison Type=::NewEnumerator0)
         True Float=-6
         False Float=-13
        )
       )
      )
      Destination Space=::Simulation
     )
    )) {enabled=false, isSetParameters=true, module=SetVariables_#10}
   /SetVariables(Particles.Position=&AddVectorToPosition(
     Position=$Particles.Position
     Vector=&Multiply_VectorByFloat(
      Vector=&NormalizeVector(Vector To Normalize=$Output.CurlNoiseForce.CurlNoiseForce, Safe Normalize=true, Safe Fallback Threshold=0, Safe Fallback Vector=<0 0 0.001>)
      Float=&FloatFromCurve(
       FloatCurve=DataInterface() {propertyValues {Curve {keys [{tangentMode=RCTM_User, time=0, value=0}, {tangentMode=RCTM_User, time=1, value=1}]}}}
       CurveIndex=&ReturnNormalizedExecIndex()
       Scale Curve=&Lerp_Float(A=1, B=2, Alpha=$User.ParaGlideLeanAlpha)
      )
     )
    )) {enabled=false, isSetParameters=true, module=SetVariables_#11}
   &ScaleColor(
    Curve Index=$Particles.RibbonLinkOrder
    Linear Color Curve=DataInterface() {propertyValues {
      AlphaCurve {keys [<at 0 1>,<at 1 1>]}
      BlueCurve {keys [<at 0.0067 0>,<at 0.7107 0>]}
      GreenCurve {keys [<at 0.0067 5>,<at 0.7107 2>]}
      RedCurve {keys [<at 0.0067 0>,<at 0.7107 0>]}
     }}
    Scale Mode=::NewEnumerator2
   ) {enabled=false}
   &ScaleRibbonWidth(Ribbon Width Scale=&FloatFromCurve(
     FloatCurve=DataInterface() {propertyValues {Curve {keys [{tangentMode=RCTM_User, time=0, value=1}, {tangentMode=RCTM_User, time=1, value=0.3}]}}}
     CurveIndex=$StackContext.RibbonLinkOrder
    ))
  ]
  rendererClasses [&NiagaraSpriteRendererProperties, &NiagaraRibbonRendererProperties]
  renderers [
   {rendererClass=&NiagaraSpriteRendererProperties}
   {rendererClass=&NiagaraRibbonRendererProperties, rendererIndex=1}
  ]
 }
 Empty001 {
  emitterUpdateScript [
   &EmitterState()
   &SpawnBurst_Instantaneous(Spawn Count=40, Use Spawn Probability=false, UseLoopCountLimit=true)
  ]
  particleSpawnScript [
   &PartitionParticles(Write Partition to Ribbon ID=::NewEnumerator0)
   /SetVariables(Particles.RibbonLinkOrder:NiagaraFloat=&ReturnNormalizedExecIndex()) {isSetParameters=true, module=SetVariables_#8}
   &InitializeParticle(
    Color=<rgba 0.2387 1 0 1>
    Lifetime=5
    Position Offset=<0 0 25>
    Ribbon Width=21
    Ribbon Width Mode=::NewEnumerator1
    Sprite Size Mode=::NewEnumerator3
    Uniform Sprite Size=5
    UsePositionOffset=false
    Write Material Random=true
   )
   &SkeletalMeshLocation(
    Bone / Socket Index=$Particles.PartitionParticles.Partition
    Bone Sampling Mode=::NewEnumerator4
    MaskSpawnGroup=false
    Sampled Position Offset Coordinate Space:ENiagaraSkelMeshTransforms=::NewEnumerator2
    Skeletal Mesh=$User.SkeletalMesh
   )
  ]
  particleUpdateScript [
   &ParticleState(
    Kill Particles When Lifetime Has Elapsed=false
    Let Infinitely Lived Particles Die When Emitter Deactivates=true
    Loop Particles Lifetime=true
   )
   &ScaleSpriteSize()
   &IncrementOverTime(Rate of Change:NiagaraFloat=&Lerp_Float(A=0.5, B=1, Alpha=$System.Dive))
   &DynamicMaterialParameters(
    Index 0 Param 1=1
    Index 0 Param 2=&MakeCustomFloatFromBool(
     Bool=&SetBoolByIntComparison(A=$Particles.PartitionParticles.Partition, B=0, Comparison Type=::NewEnumerator0)
     True Float=-1
     False Float=-0.8
    )
    Index 0 Param 3=&MakeCustomFloatFromBool(
     Bool=&SetBoolByIntComparison(A=$Particles.PartitionParticles.Partition, B=0, Comparison Type=::NewEnumerator0)
     True Float=-0.8
     False Float=-1
    )
    Index 0 Param 4=&FloatFromCurve(FloatCurve=DataInterface() {propertyValues {Curve {keys [<at 0 1>,<at 1 0>]}}}, CurveIndex=$Particles.RibbonLinkOrder, Scale Curve=0.25)
    Module.Param0WriteEnabled=true
    Module.Param1WriteEnabled=true
    Module.Param2WriteEnabled=true
    Module.Param3WriteEnabled=true
   )
   &SkeletalMeshLocation(
    Bone / Socket Index=$Particles.PartitionParticles.Partition
    Bone Sampling Mode=::NewEnumerator4
    MaskSpawnGroup=false
    Sampled Position Offset Coordinate Space:ENiagaraSkelMeshTransforms=::NewEnumerator2
    Skeletal Mesh=$User.SkeletalMesh
   ) {module=SkeletalMeshLocation001}
   /SetVariables(Particles.Position=&AddVectorToPosition(
     Position=$Particles.Position
     Vector=&Multiply_VectorByFloat(
      Vector=$System.DiveVelocityScale
      Float=&FloatFromCurve(
       FloatCurve=DataInterface() {propertyValues {Curve {keys [{tangentMode=RCTM_User, time=0, value=0}, {tangentMode=RCTM_User, time=1, value=1}]}}}
       CurveIndex=&ReturnNormalizedExecIndex()
       Scale Curve=&Lerp_Float(A=-100, B=-130, Alpha=$System.Dive)
      )
     )
    )) {isSetParameters=true, module=SetVariables_#9}
   /SetVariables(Particles.Position=&AddVectorToPosition(
     Position=$Particles.Position
     Vector=&TransformVector(
      Vector=&Multiply_VectorByFloat(
       Vector=&SelectVectorFromArray(
        Direct Array Index=$Particles.PartitionParticles.Partition
        Vector Selection Array=DataInterface() {propertyValues {FloatData [<0.75 0 0>, <-0.75 0 0>, <0.75 0 0>, <-0.75 0 0>]}}
       )
       Float=&FloatFromCurve(
        FloatCurve=DataInterface() {propertyValues {Curve {keys [
            {arriveTangent=4.1785, interpMode=RCIM_Cubic, leaveTangent=4.1785, tangentMode=RCTM_User, time=0, value=0}
            {arriveTangent=1.1179, interpMode=RCIM_Cubic, leaveTangent=1.1179, tangentMode=RCTM_User, time=0.2042, value=0.6228}
            {arriveTangent=0.0224, interpMode=RCIM_Cubic, leaveTangent=0.0224, tangentMode=RCTM_User, time=1, value=1}
           ]}}}
        CurveIndex=&ReturnNormalizedExecIndex()
        Scale Curve=&MakeCustomFloatFromBool(
         Bool=&SetBoolByIntComparison(A=$Particles.PartitionParticles.Partition, B=1, Comparison Type=::NewEnumerator0)
         True Float=-6
         False Float=-13
        )
       )
      )
      Destination Space=::Simulation
     )
    )) {enabled=false, isSetParameters=true, module=SetVariables_#10}
   /SetVariables(Particles.Position=&AddVectorToPosition(
     Position=$Particles.Position
     Vector=&Multiply_VectorByFloat(
      Vector=&NormalizeVector(Vector To Normalize=$Output.CurlNoiseForce.CurlNoiseForce, Safe Normalize=true, Safe Fallback Threshold=0, Safe Fallback Vector=<0 0 0.001>)
      Float=&FloatFromCurve(
       FloatCurve=DataInterface() {propertyValues {Curve {keys [{tangentMode=RCTM_User, time=0, value=0}, {tangentMode=RCTM_User, time=1, value=1}]}}}
       CurveIndex=&ReturnNormalizedExecIndex()
       Scale Curve=&Lerp_Float(A=1, B=2, Alpha=$User.ParaGlideLeanAlpha)
      )
     )
    )) {enabled=false, isSetParameters=true, module=SetVariables_#11}
   &ScaleColor(
    Curve Index=$Particles.RibbonLinkOrder
    Linear Color Curve=DataInterface() {propertyValues {
      AlphaCurve {keys [<at 0 1>,<at 1 1>]}
      BlueCurve {keys [<at 0.0067 0>,<at 0.7107 0>]}
      GreenCurve {keys [<at 0.0067 5>,<at 0.7107 2>]}
      RedCurve {keys [<at 0.0067 0>,<at 0.7107 0>]}
     }}
    Scale Mode=::NewEnumerator2
   ) {enabled=false}
   &ScaleRibbonWidth(Ribbon Width Scale=&FloatFromCurve(
     FloatCurve=DataInterface() {propertyValues {Curve {keys [{tangentMode=RCTM_User, time=0, value=1}, {tangentMode=RCTM_User, time=1, value=0.5}]}}}
     CurveIndex=$StackContext.RibbonLinkOrder
    ))
  ]
  rendererClasses [&NiagaraSpriteRendererProperties, &NiagaraRibbonRendererProperties]
  renderers [
   {rendererClass=&NiagaraSpriteRendererProperties}
   {rendererClass=&NiagaraRibbonRendererProperties, rendererIndex=1}
  ]
 }
 Empty002 {
  emitterUpdateScript [
   &EmitterState()
   &SpawnBurst_Instantaneous(Spawn Count=40, Use Spawn Probability=false, UseLoopCountLimit=true)
  ]
  particleSpawnScript [
   &PartitionParticles(Write Partition to Ribbon ID=::NewEnumerator0)
   /SetVariables(Particles.RibbonLinkOrder:NiagaraFloat=&ReturnNormalizedExecIndex()) {isSetParameters=true, module=SetVariables_#8}
   &InitializeParticle(
    Color=<rgba 0.2387 1 0 1>
    Lifetime=5
    Position Offset=<0 0 25>
    Ribbon Width=25
    Ribbon Width Mode=::NewEnumerator1
    Sprite Size Mode=::NewEnumerator3
    Uniform Sprite Size=5
    UsePositionOffset=false
    Write Material Random=true
   )
   &SkeletalMeshLocation(
    Bone / Socket Index=&SelectIntFromArray(Direct Array Index=$Particles.PartitionParticles.Partition, Int Selection Array=DataInterface() {propertyValues {IntData [4,5]}})
    Bone Sampling Mode=::NewEnumerator4
    MaskSpawnGroup=false
    Sampled Position Offset Coordinate Space:ENiagaraSkelMeshTransforms=::NewEnumerator2
    Skeletal Mesh=$User.SkeletalMesh
   )
  ]
  particleUpdateScript [
   &ParticleState(
    Kill Particles When Lifetime Has Elapsed=false
    Let Infinitely Lived Particles Die When Emitter Deactivates=true
    Loop Particles Lifetime=true
   )
   &ScaleSpriteSize()
   &IncrementOverTime(Rate of Change:NiagaraFloat=&Lerp_Float(A=0.3, B=0.7, Alpha=$System.Dive))
   &DynamicMaterialParameters(
    Index 0 Param 1=1
    Index 0 Param 2=&MakeCustomFloatFromBool(
     Bool=&SetBoolByIntComparison(A=$Particles.PartitionParticles.Partition, B=0, Comparison Type=::NewEnumerator0)
     True Float=-0.7
     False Float=-1
    )
    Index 0 Param 3=&MakeCustomFloatFromBool(
     Bool=&SetBoolByIntComparison(A=$Particles.PartitionParticles.Partition, B=0, Comparison Type=::NewEnumerator0)
     True Float=-1
     False Float=-0.8
    )
    Index 0 Param 4=&FloatFromCurve(FloatCurve=DataInterface() {propertyValues {Curve {keys [<at 0 1>,<at 1 0>]}}}, CurveIndex=$Particles.RibbonLinkOrder, Scale Curve=0.2)
    Module.Param0WriteEnabled=true
    Module.Param1WriteEnabled=true
    Module.Param2WriteEnabled=true
    Module.Param3WriteEnabled=true
   )
   &SkeletalMeshLocation(
    Bone / Socket Index=&SelectIntFromArray(Direct Array Index=$Particles.PartitionParticles.Partition, Int Selection Array=DataInterface() {propertyValues {IntData [4,5]}})
    Bone Sampling Mode=::NewEnumerator4
    MaskSpawnGroup=false
    Sampled Position Offset=<0 -5 0>
    Sampled Position Offset Coordinate Space:ENiagaraSkelMeshTransforms=::NewEnumerator2
    Skeletal Mesh=$User.SkeletalMesh
   ) {module=SkeletalMeshLocation001}
   /SetVariables(Particles.Position=&AddVectorToPosition(
     Position=$Particles.Position
     Vector=&Multiply_VectorByFloat(
      Vector=$System.DiveVelocityScale
      Float=&FloatFromCurve(
       FloatCurve=DataInterface() {propertyValues {Curve {keys [{tangentMode=RCTM_User, time=0, value=0}, {tangentMode=RCTM_User, time=1, value=1}]}}}
       CurveIndex=&ReturnNormalizedExecIndex()
       Scale Curve=&Lerp_Float(A=-200, B=-300, Alpha=$System.Dive)
      )
     )
    )) {isSetParameters=true, module=SetVariables_#9}
   /SetVariables(Particles.Position=&AddVectorToPosition(
     Position=$Particles.Position
     Vector=&TransformVector(
      Vector=&Multiply_VectorByFloat(
       Vector=&SelectVectorFromArray(
        Direct Array Index=$Particles.PartitionParticles.Partition
        Vector Selection Array=DataInterface() {propertyValues {FloatData [<0.75 0 0>, <-0.75 0 0>, <0.75 0 0>, <-0.75 0 0>]}}
       )
       Float=&FloatFromCurve(
        FloatCurve=DataInterface() {propertyValues {Curve {keys [
            {arriveTangent=4.1785, interpMode=RCIM_Cubic, leaveTangent=4.1785, tangentMode=RCTM_User, time=0, value=0}
            {arriveTangent=1.1179, interpMode=RCIM_Cubic, leaveTangent=1.1179, tangentMode=RCTM_User, time=0.2042, value=0.6228}
            {arriveTangent=0.0224, interpMode=RCIM_Cubic, leaveTangent=0.0224, tangentMode=RCTM_User, time=1, value=1}
           ]}}}
        CurveIndex=&ReturnNormalizedExecIndex()
        Scale Curve=&MakeCustomFloatFromBool(
         Bool=&SetBoolByIntComparison(A=$Particles.PartitionParticles.Partition, B=1, Comparison Type=::NewEnumerator0)
         True Float=-6
         False Float=-13
        )
       )
      )
      Destination Space=::Simulation
     )
    )) {enabled=false, isSetParameters=true, module=SetVariables_#10}
   /SetVariables(Particles.Position=&AddVectorToPosition(
     Position=$Particles.Position
     Vector=&Multiply_VectorByFloat(
      Vector=&NormalizeVector(Vector To Normalize=$Output.CurlNoiseForce.CurlNoiseForce, Safe Normalize=true, Safe Fallback Threshold=0, Safe Fallback Vector=<0 0 0.001>)
      Float=&FloatFromCurve(
       FloatCurve=DataInterface() {propertyValues {Curve {keys [{tangentMode=RCTM_User, time=0, value=0}, {tangentMode=RCTM_User, time=1, value=1}]}}}
       CurveIndex=&ReturnNormalizedExecIndex()
       Scale Curve=&Lerp_Float(A=1, B=2, Alpha=$User.ParaGlideLeanAlpha)
      )
     )
    )) {enabled=false, isSetParameters=true, module=SetVariables_#11}
   &ScaleColor(
    Curve Index=$Particles.RibbonLinkOrder
    Linear Color Curve=DataInterface() {propertyValues {
      AlphaCurve {keys [<at 0 1>,<at 1 1>]}
      BlueCurve {keys [<at 0.0067 0>,<at 0.7107 0>]}
      GreenCurve {keys [<at 0.0067 5>,<at 0.7107 2>]}
      RedCurve {keys [<at 0.0067 0>,<at 0.7107 0>]}
     }}
    Scale Mode=::NewEnumerator2
   ) {enabled=false}
   &ScaleRibbonWidth(Ribbon Width Scale=&FloatFromCurve(
     FloatCurve=DataInterface() {propertyValues {Curve {keys [{tangentMode=RCTM_User, time=0, value=1}, {tangentMode=RCTM_User, time=1, value=0.6}]}}}
     CurveIndex=$StackContext.RibbonLinkOrder
    ))
  ]
  rendererClasses [&NiagaraSpriteRendererProperties, &NiagaraRibbonRendererProperties]
  renderers [
   {rendererClass=&NiagaraSpriteRendererProperties}
   {rendererClass=&NiagaraRibbonRendererProperties, rendererIndex=1}
  ]
 }
 Forward_smoke {
  emitterUpdateScript [
   &EmitterState()
   &SpawnRate(
    SpawnRate=&MakeCustomFloatFromBool(Bool=&SetBoolByFloatComparison(A=$System.Dive, B=0.8), True Float=15, False Float=0)
    Use Spawn Probability=false
   )
  ]
  particleSpawnScript [
   &InitializeParticle(
    Color=&MakeCustomLinearColorFromBool(Bool=&RandomBool(), True Color=<rgba 0.0292 1.2 0 0.2>, False Color=<rgba 0.9123 1.2 0 0.2>)
    Lifetime Max=0.7
    Lifetime Min=0.5
    Lifetime Mode=::NewEnumerator1
    Sprite Size Max=<300 450>
    Sprite Size Min=<150 250>
    Sprite Size Mode=::NewEnumerator2
    Sprite UV Mode=::NewEnumerator3
    UsePositionOffset=true
    Write Material Random=true
   )
   &ShapeLocation(
    Handle Radius=20
    Large Radius=120
    MaskSpawnGroup=false
    Offset=&FN_FrontEndVector3Modifier(Gameplay=<0 -25 -100>, Front End=<0 200 -200>)
    Offset Coordinate Space=::Local
    Offset Mode=::NewEnumerator0
    Override Seed=false
    Shape Primitive=::NewEnumerator3
    UseOwnerScale=true
   )
   &DynamicMaterialParameters(
    Index 0 Param 1=&RandomRangeFloat(Minimum=0, Maximum=0.3)
    Module.Param0WriteEnabled=true
    Module.Param1WriteEnabled=true
    Module.Param2WriteEnabled=true
    Module.Param3WriteEnabled=true
   )
   &AddVelocity(Velocity=&FN_FrontEndVector3Modifier(
     Gameplay=<0 0 0>
     Front End=&MakeVector(X=0, Y=&Lerp_Float(A=0, B=-200, Alpha=$System.Dive), Z=300)
    ))
  ]
  particleUpdateScript [
   &ParticleState()
   &InheritSourceMovement(
    Applied Position Scale=&MakeVector(X=0.5, Y=0.5, Z=0.5)
    ApplyForce=false
    ApplyPosition=true
    ApplyPositionFalloff=false
   ) {enabled=false, module=InheritSourceMovement001}
   &SpriteRotationRate(Rotation Rate=&RandomRangeFloat(Minimum=-50, Maximum=50))
   &ScaleSpriteSize(
    Scale Factor=&Vector2DFromCurve(
     Vector2Curve=DataInterface() {propertyValues {
       XCurve {keys [{tangentMode=RCTM_User, time=0, value=0}, {tangentMode=RCTM_User, time=1, value=1}]}
       YCurve {keys [{tangentMode=RCTM_User, time=0, value=0.8}, {tangentMode=RCTM_User, time=1, value=1}]}
      }}
     ScaleCurve=<1 1>
    )
    Scale Sprite Size Mode=::NewEnumerator2
   )
   &ScaleSpriteSize(Scale Factor=<1 5>, Scale Sprite Size Mode=::NewEnumerator2) {module=ScaleSpriteSize001}
   &Drag(Drag=5, Use Linear Drag=true, Use Rotational Drag=true) {enabled=false}
   &InheritVelocity(
    Speed Limit=true
    UseSpeedThreshold=false
    Velocity Source=&Multiply_VectorByFloat(Vector=$Engine.Owner.Velocity, Float=0.2)
   )
   &SolveForcesAndVelocity(Clamp Velocity=false, Limit Acceleration=false)
   &ScaleSpriteSizeBySpeed(Max Scale Factor=<1 5>) {enabled=false}
   &ScaleColor(
    Scale Alpha=&FloatFromCurve(
     FloatCurve=DataInterface() {propertyValues {Curve {keys [
         {interpMode=RCIM_Cubic, tangentMode=RCTM_User, time=0, value=0}
         {interpMode=RCIM_Cubic, leaveTangent=-3.074, tangentMode=RCTM_Break, time=0.3, value=1}
         {tangentMode=RCTM_User, time=1, value=0}
        ]}}}
     Scale Curve=&FloatFromCurve(
      FloatCurve=DataInterface() {propertyValues {Curve {keys [
          {interpMode=RCIM_Cubic, tangentMode=RCTM_User, time=0, value=0}
          {interpMode=RCIM_Cubic, leaveTangent=-1.5986, tangentMode=RCTM_Break, time=0.3, value=1}
          {tangentMode=RCTM_User, time=1, value=0}
         ]}}}
      Scale Curve=&FN_FrontEndFloatModifier(Gameplay=0.8, Front End=1.5)
     )
    )
    Scale RGB=&Multiply_VectorByFloat(Vector=<1 1 1>, Float=&FN_FrontEndFloatModifier(Gameplay=0.9, Front End=0.8))
    ScaleA=true
    ScaleRGB=true
   )
  ]
  rendererClasses [&NiagaraSpriteRendererProperties]
  renderers [
   {rendererClass=&NiagaraSpriteRendererProperties}
   {rendererClass=&NiagaraSpriteRendererProperties, rendererIndex=1}
  ]
 }
 Logo_UnderPlayer {
  emitterUpdateScript [
   &EmitterState(Life Cycle Mode=::NewEnumerator1, Loop Behavior=::NewEnumerator1, Loop Duration Mode=::NewEnumerator1, UseLoopDelay=false)
   &SpawnBurst_Instantaneous(Use Spawn Probability=false, UseLoopCountLimit=false)
  ]
  particleSpawnScript [
   &InitializeParticle(
    Color=<rgba 0.6614 5 0 1.5>
    Lifetime=10
    Mesh Scale=<1.7 1.7 1>
    Mesh Scale Mode=::NewEnumerator4
    Position Offset=<0 0 -200>
    Sprite Size Mode=::NewEnumerator3
    Uniform Sprite Size=250
    UsePositionOffset=false
    Write Material Random=true
   )
   &InitialMeshOrientation(
    Mesh Orientation Mode=::NewEnumerator2
    Orientation Coordinate Space=::Local
    Orientation Vector=<0 0 1>
    Rotation=<0.55 0 0>
    Rotation Coordinate Space=::NewEnumerator2
    Use Rotation Vector=true
   )
  ]
  particleUpdateScript [
   &ParticleState(Kill Particles When Lifetime Has Elapsed=false, Let Infinitely Lived Particles Die When Emitter Deactivates=true)
   &ScaleMeshSize(Scale Factor=&MakeVector(X=1, Y=1, Z=&Lerp_Float(A=0.25, B=1.25, Alpha=$System.Timeline001.Float)))
   &UpdateMeshOrientation(
    Facing Direction=$System.DiveVelocityScale
    Facing Interpolation=false
    Orientation Method=::NewEnumerator1
    Override Quat=false
    TransformFacing=false
    TransformUp=false
    Up Direction:Vector3f=$Engine.Owner.SystemXAxis
   ) {enabled=false}
   &SystemLocation(MaskSpawnGroup=false, Offset=&FN_FrontEndVector3Modifier(Gameplay=<0 0 40>, Front End=<0 50 40>))
   &SkeletalMeshLocation(
    Bone / Socket Index=4
    Bone Sampling Mode=::NewEnumerator4
    MaskSpawnGroup=false
    Orientation Sampling=::NewEnumerator1
    Sampled Position Offset=<0 0 50>
    Skeletal Mesh=$User.Skeletal Mesh
   ) {enabled=false}
   /SetVariables(Particles.Light_Position:NiagaraPosition=&AddVectorToPosition(
     Position=$Particles.SkeletalMeshLocation.SampledPosition
     Vector=&Lerp_Vector(A=<0 0 0>, B=<0 0 -40>, Alpha=$System.Dive)
    )) {isSetParameters=true}
   /SetVariables(Particles.MeshOrientation=&OrientMeshToVector(
     Look At Direction=&FN_FrontEndVector3Modifier(
      Gameplay=&Multiply_VectorByFloat(
       Vector=&NormalizeVector(
        Vector To Normalize=&NegateVector(Vector:Vector3f=&TransformVector(
          Vector=&Multiply_Vector(A=$Engine.Owner.Velocity, B=<0.5 0.5 1>)
          Source Space=::World
          Destination Space=::Local
         ))
        Safe Normalize=true
        Safe Fallback Threshold=0
        Safe Fallback Vector=<0 0 0.001>
       )
       Float=-1
      )
      Front End=<0 0.25 -1>
     )
     Reference Axis Vector=<1 0 0>
    )) {isSetParameters=true}
   &DynamicMaterialParameters(
    Index 0 Param 2=2
    Index 0 Param 3=1
    Index 0 Param 4=0.65
    Module.Param0WriteEnabled=true
    Module.Param0WriteEnabled1=true
    Module.Param1WriteEnabled=true
    Module.Param1WriteEnabled1=true
    Module.Param2WriteEnabled=true
    Module.Param2WriteEnabled1=true
    Module.Param3WriteEnabled=true
    Module.Param3WriteEnabled1=true
    Write Parameter Index 1=true
   )
   &SolveForcesAndVelocity(Clamp Velocity=false, Limit Acceleration=false)
   &FN_Light_Attributes(Color=<rgba 1.1537 5 0 1>, Create Color Attribute=true, Create Exponent Attribute=true, Create Radius Attribute=true, Exponent=3, Radius=75)
  ]
  rendererClasses [&NiagaraMeshRendererProperties, &NiagaraLightRendererProperties]
  renderers [
   {rendererClass=&NiagaraMeshRendererProperties, rendererIndex=0}
   {rendererClass=&NiagaraMeshRendererProperties, rendererIndex=1}
   {rendererClass=&NiagaraLightRendererProperties, rendererIndex=2}
  ]
 }
 Rotating_Keys001 {
  emitterUpdateScript [
   &EmitterState(Life Cycle Mode=::NewEnumerator1, Loop Behavior=::NewEnumerator1, Loop Duration Mode=::NewEnumerator1, UseLoopDelay=false)
   &SpawnBurst_Instantaneous(Spawn Count=2, Use Spawn Probability=false, UseLoopCountLimit=false)
  ]
  particleSpawnScript [
   /SetVariables(Particles.CameraOffset_Glow=5) {isSetParameters=true, module=SetVariables_#12}
   &PartitionParticles()
   /SetVariables(Particles.SpawnGroup=$Particles.PartitionParticles.Partition) {isSetParameters=true, module=SetVariables_#13}
   &InitializeParticle(
    Color=<rgba 5 1.9885 0 1>
    Lifetime=5
    Mesh Scale Mode=::NewEnumerator3
    Mesh Uniform Scale=0.6
    UsePositionOffset=false
    Write Material Random=true
   )
   &InitialMeshOrientation(Mesh Axis To Orient=<0 0 1>, Mesh Orientation Mode=::NewEnumerator2, Orientation Coordinate Space=::Local, Use Rotation Vector=false)
  ]
  particleUpdateScript [
   &SolveForcesAndVelocity(Clamp Velocity=false, Limit Acceleration=false)
   &ParticleState(
    Kill Particles When Lifetime Has Elapsed=false
    Let Infinitely Lived Particles Die When Emitter Deactivates=true
    Loop Particles Lifetime=true
   )
   /SetVariables(Particles.Rotate_Vector:Vector3f=&Multiply_VectorByFloat(
     Vector=&NormalizeVector(
      Vector To Normalize=&NegateVector(Vector:Vector3f=&TransformVector(
        Vector=&Multiply_Vector(A=$Engine.Owner.Velocity, B=<0.5 0.5 1>)
        Source Space=::World
        Destination Space=::Local
       ))
      Safe Normalize=true
      Safe Fallback Threshold=0
      Safe Fallback Vector=<0 0 0.001>
     )
     Float=&Lerp_Float(A=150, B=200, Alpha=$System.Dive)
    )) {isSetParameters=true, module=SetVariables_#14}
   &ShapeLocation(
    MaskSpawnGroup=true
    Offset=$System.Sphere_RotateVector
    Offset Coordinate Space=::Local
    Offset Mode=::NewEnumerator0
    Rotation Axis=$Particles.Rotate_Vector
    Rotation Mode=::NewEnumerator1
    Sphere Distribution=::NewEnumerator1
    Sphere Radius=$System.Sphere_Radius
    Sphere Radius Position:NiagaraFloat=&Multiply_Float(A=$System.KeyCurve, B=-1)
    U Position=&Multiply_Float(A=$Engine.System.Age, B=$System.UPos_Time)
    UseOwnerScale=true
   )
   &ShapeLocation(
    MaskSpawnGroup=true
    Offset=$System.Sphere_RotateVector
    Offset Coordinate Space=::Local
    Offset Mode=::NewEnumerator0
    Rotation Axis=$Particles.Rotate_Vector
    Rotation Mode=::NewEnumerator1
    Spawn Group Mask=1
    Sphere Distribution=::NewEnumerator1
    Sphere Radius=$System.Sphere_Radius
    Sphere Radius Position:NiagaraFloat=$System.KeyCurve
    U Position=&Multiply_Float(A=$Emitter.Age, B=$System.UPos_Time)
    UseOwnerScale=true
   ) {module=ShapeLocation001}
   &CalculateAccurateVelocity() {enabled=false}
   &UpdateMeshOrientation(
    Facing Interpolation=false
    Facing Position:NiagaraPosition=&MaskVectorBySpawnGroup(
     Masked Vector=$Particles.ShapeLocation001.ShapePosition
     Passthrough Vector=Linked() {linkedVariable {name=Particles.ShapeLocation.ShapePosition, type=NiagaraPosition}}
    )
    Orientation Method=::NewEnumerator2
    Override Quat=false
    TransformFacingOrigin=false
    TransformFacingPosition=false
    TransformUp=false
   )
   &ScaleColor(
    Scale Alpha=&FloatFromCurve(
     FloatCurve=DataInterface() {propertyValues {Curve {keys [{interpMode=RCIM_Cubic, tangentMode=RCTM_User, time=0, value=0.25}, {interpMode=RCIM_Cubic, tangentMode=RCTM_User, time=1, value=1}]}}}
     CurveIndex=$System.ScalingCurve1
     Scale Curve=&Waveform("Waveform [1]"=::NewEnumerator9, "[1] Phase Scale"=2, "[1] Frequency"=5, "[1] Amplitude Min/Max"=<0.25 1>)
    )
    ScaleA=true
    ScaleRGB=true
   )
   &FN_Light_Attributes(
    Color=<rgba 5 1.5059 0 1>
    Create Color Attribute=true
    Create Exponent Attribute=true
    Create Radius Attribute=true
    Exponent=3
    Radius=75
    Scale Color=&FloatFromCurve(
     FloatCurve=DataInterface() {propertyValues {Curve {keys [{interpMode=RCIM_Cubic, tangentMode=RCTM_User, time=0, value=0}, {interpMode=RCIM_Cubic, tangentMode=RCTM_User, time=1, value=0.5}]}}}
     CurveIndex=$System.ScalingCurve1
     Scale Curve=&Waveform("Waveform [1]"=::NewEnumerator9, "[1] Phase Scale"=2, "[1] Frequency"=5, "[1] Amplitude Min/Max"=<0.25 1>, "[1] Scale"=3)
    )
   )
  ]
  rendererClasses [
   &NiagaraMeshRendererProperties
   &NiagaraSpriteRendererProperties
   &NiagaraLightRendererProperties
  ]
  renderers [
   {rendererClass=&NiagaraMeshRendererProperties, rendererIndex=0}
   {rendererClass=&NiagaraSpriteRendererProperties, rendererIndex=1}
   {rendererClass=&NiagaraLightRendererProperties, rendererIndex=2}
  ]
 }
 Spline_ConnectingKeys {
  emitterUpdateScript [
   &EmitterState(Life Cycle Mode=::NewEnumerator1, Loop Behavior=::NewEnumerator1, Loop Duration Mode=::NewEnumerator1, UseLoopDelay=false)
   &SpawnBurst_Instantaneous(Spawn Count=2, Use Spawn Probability=false, UseLoopCountLimit=false)
  ]
  particleSpawnScript [
   /SetVariables(Particles.CameraOffset_Glow=5) {isSetParameters=true, module=SetVariables_#12}
   &PartitionParticles()
   /SetVariables(Particles.SpawnGroup=$Particles.PartitionParticles.Partition) {isSetParameters=true, module=SetVariables_#13}
   &InitializeParticle(
    Color=<rgba 5 1.9885 0 1>
    Lifetime=5
    Mesh Scale=<0.66 0.68 1>
    Mesh Scale Mode=::NewEnumerator4
    UsePositionOffset=false
    Write Material Random=true
   )
   &InitialMeshOrientation(Mesh Orientation Mode=::NewEnumerator2, Orientation Coordinate Space=::Local, Use Rotation Vector=false)
  ]
  particleUpdateScript [
   /SetVariables(Particles.SpriteDynamicParams:Vector4f=<0 0 15 0>) {isSetParameters=true}
   &SolveForcesAndVelocity(Clamp Velocity=false, Limit Acceleration=false)
   &ParticleState(
    Kill Particles When Lifetime Has Elapsed=false
    Let Infinitely Lived Particles Die When Emitter Deactivates=true
    Loop Particles Lifetime=true
   )
   &ScaleMeshSize(Scale Factor=&Multiply_VectorByFloat(
     Vector=<1 1 1>
     Float=&FloatFromCurve(
      FloatCurve=DataInterface() {propertyValues {Curve {keys [{tangentMode=RCTM_User, time=0, value=0.05}, {tangentMode=RCTM_User, time=1, value=1}]}}}
      CurveIndex=$System.KeyCurve
     )
    ))
   /SetVariables(Particles.Rotate_Vector:Vector3f=&Multiply_VectorByFloat(
     Vector=&NormalizeVector(
      Vector To Normalize=&NegateVector(Vector:Vector3f=&TransformVector(
        Vector=&Multiply_Vector(A=$Engine.Owner.Velocity, B=<0.5 0.5 1>)
        Source Space=::World
        Destination Space=::Local
       ))
      Safe Normalize=true
      Safe Fallback Threshold=0
      Safe Fallback Vector=<0 0 0.001>
     )
     Float=&Lerp_Float(A=150, B=200, Alpha=$System.Dive)
    )) {isSetParameters=true, module=SetVariables_#14}
   &ShapeLocation(
    MaskSpawnGroup=true
    Offset=$System.Sphere_RotateVector
    Offset Coordinate Space=::Local
    Offset Mode=::NewEnumerator0
    Rotation Axis=$Particles.Rotate_Vector
    Rotation Mode=::NewEnumerator1
    Sphere Distribution=::NewEnumerator1
    Sphere Radius=$System.Sphere_Radius
    Sphere Radius Position:NiagaraFloat=&Multiply_Float(A=$System.KeyCurve, B=-1)
    U Position=&Multiply_Float(A=$Engine.System.Age, B=$System.UPos_Time)
    UseOwnerScale=true
   )
   &ShapeLocation(
    MaskSpawnGroup=true
    Offset=$System.Sphere_RotateVector
    Offset Coordinate Space=::Local
    Offset Mode=::NewEnumerator0
    Rotation Axis=$Particles.Rotate_Vector
    Rotation Mode=::NewEnumerator1
    Spawn Group Mask=1
    Sphere Distribution=::NewEnumerator1
    Sphere Radius=$System.Sphere_Radius
    Sphere Radius Position:NiagaraFloat=$System.KeyCurve
    U Position=&Multiply_Float(A=$Emitter.Age, B=$System.UPos_Time)
    UseOwnerScale=true
   ) {module=ShapeLocation001}
   &CalculateAccurateVelocity() {enabled=false}
   &UpdateMeshOrientation(
    Facing Interpolation=false
    Facing Position:NiagaraPosition=&MaskVectorBySpawnGroup(
     Masked Vector=$Particles.ShapeLocation001.ShapePosition
     Passthrough Vector=Linked() {linkedVariable {name=Particles.ShapeLocation.ShapePosition, type=NiagaraPosition}}
    )
    Orientation Method=::NewEnumerator2
    Override Quat=false
    TransformFacingOrigin=false
    TransformFacingPosition=false
    TransformUp=false
   )
   &ScaleColor(
    Scale Alpha=&FloatFromCurve(
     FloatCurve=DataInterface() {propertyValues {Curve {keys [{interpMode=RCIM_Cubic, tangentMode=RCTM_User, time=0, value=0}, {interpMode=RCIM_Cubic, tangentMode=RCTM_User, time=1, value=0.5}]}}}
     CurveIndex=$System.ScalingCurve1
    )
    ScaleA=true
    ScaleRGB=true
   )
   &FN_Light_Attributes(
    Color=<rgba 5 1.5059 0 1>
    Create Color Attribute=true
    Create Exponent Attribute=true
    Create Radius Attribute=true
    Exponent=3
    Radius=75
    Scale Color=&FloatFromCurve(
     FloatCurve=DataInterface() {propertyValues {Curve {keys [{interpMode=RCIM_Cubic, tangentMode=RCTM_User, time=0, value=0}, {interpMode=RCIM_Cubic, tangentMode=RCTM_User, time=1, value=0.5}]}}}
     CurveIndex=$System.ScalingCurve1
     Scale Curve=3
    )
   )
  ]
  rendererClasses [&NiagaraMeshRendererProperties]
  renderers [
   {rendererClass=&NiagaraMeshRendererProperties, rendererIndex=0}
   {rendererClass=&NiagaraMeshRendererProperties, rendererIndex=1}
  ]
 }
}
summary {
 systemName=NS_Contrail_LoudPhoenix1
 emitters [
  {
   emitterName=Logo_UnderPlayer
   rendererClasses [&NiagaraMeshRendererProperties, &NiagaraLightRendererProperties]
  }
  {
   emitterName=CenterGlow_for_Keys
   rendererClasses [&NiagaraSpriteRendererProperties, &NiagaraLightRendererProperties]
  }
  {
   emitterName=Rotating_Keys001
   rendererClasses [
    &NiagaraMeshRendererProperties
    &NiagaraSpriteRendererProperties
    &NiagaraLightRendererProperties
   ]
  }
  {emitterName=Spline_ConnectingKeys, rendererClasses [&NiagaraMeshRendererProperties]}
  {
   emitterName=CenterGlow_for_Keys001
   rendererClasses [
    &NiagaraSpriteRendererProperties
    &NiagaraMeshRendererProperties
    &NiagaraRibbonRendererProperties
   ]
  }
  {emitterName=Forward_smoke, rendererClasses [&NiagaraSpriteRendererProperties]}
  {emitterName=Empty001, rendererClasses [&NiagaraSpriteRendererProperties, &NiagaraRibbonRendererProperties]}
  {emitterName=Empty, rendererClasses [&NiagaraSpriteRendererProperties, &NiagaraRibbonRendererProperties]}
  {emitterName=Empty002, rendererClasses [&NiagaraSpriteRendererProperties, &NiagaraRibbonRendererProperties]}
 ]
 userVariables [
  User.bIsFrontEnd {type=NiagaraBool}
  User.bIsFrontEndPreview {type=NiagaraBool, defaultValue=true}
  User.ForwardDot {type=NiagaraFloat, defaultValue=0, description=""}
  User.LocalHorizontalDistance {type=NiagaraFloat, defaultValue=0, description=""}
  User.LocalHorizontalVector {type=Vector3f, description="", defaultValue {struct=Vector3f, value=<0 0 0>}}
  User.ParaGlideLeanAlpha {type=NiagaraFloat, defaultValue=0, description=""}
  User.PlayerSpeed {type=NiagaraFloat, defaultValue=0, description=""}
  User.RibbonIDOffset {type=NiagaraInt32}
  User.RightDot {type=NiagaraFloat, defaultValue=0, description=""}
  User.Skeletal Mesh {
   type=NiagaraDataInterfaceSkeletalMesh
   defaultValue {
    struct=NiagaraExt_VariableValue_DataInterface
    value {dataInterface=&self:NiagaraDataInterfaceSkeletalMesh_2}
   }
  }
  User.SkeletalMesh {
   type=NiagaraDataInterfaceSkeletalMesh
   defaultValue {
    struct=NiagaraExt_VariableValue_DataInterface
    value {dataInterface=&self:NiagaraDataInterfaceSkeletalMesh_0}
   }
  }
  User.TimeOffset {type=NiagaraFloat, defaultValue=0, description=""}
  User.Variant_Color {type=LinearColor, description="", defaultValue {struct=LinearColor, value=<rgba 1 1 1 1>}}
  User.VariantID {type=NiagaraFloat, defaultValue=0, description=""}
 ]
}
